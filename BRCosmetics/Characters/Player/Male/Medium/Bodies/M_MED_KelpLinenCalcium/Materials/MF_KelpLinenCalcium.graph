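# Material function for the KelpLinenCalcium player cosmetic that rewrites base color and emissive with an animated, audio-reactive glow effect.
# Mask textures and a UV sphere mask isolate shirt, logo, small details, gradient and dot regions.
# Panned, distorted noise in pre-skinned space and time-driven sine waves feed masked curve-atlas coloring, adjusted by time of day.
# An audio reactivity function, given a synthesized frontend fallback loop, supplies the amplitude.
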MaterialFunction /BRCosmetics/Characters/Player/Male/Medium/Bodies/M_MED_KelpLinenCalcium/Materials/MF_KelpLinenCalcium
graph {nodes [
  ComponentMask(None=NamedRerouteUsage()) {b=true}
  n1 = GetMaterialAttributes(None=NamedRerouteUsage()) {attributeGetTypes [#1, #2]}
  FunctionOutput(None=SetMaterialAttributes!(MaterialAttributes=@n1.MaterialAttributes, BaseColor=NamedRerouteUsage(), EmissiveColor=NamedRerouteUsage()) {attributeSetTypes [#1, #2]})
  n2 = Multiply(A=Time()) {constB=0.075}
  LinearInterpolate(A=@n2, B=Multiply(A=@n2) {constB=1.1}, Alpha=NamedRerouteUsage())
  NamedRerouteDeclaration(None=FunctionInput() {inputType=FunctionInput_MaterialAttributes}) {name=InMaterial, nodeColor=<rgba 1 0 0.9609 1>}
  NamedRerouteDeclaration(None=FunctionInput() {bUsePreviewValueAsDefault=true, inputName=FXMask, sortPriority=2, previewValue=<1 1 1 1>}) {name=FXMask, nodeColor=<rgba 0 1 0.7188 1>}
  n3 = NamedRerouteUsage()
  n4 = NamedRerouteUsage()
  n5 = Add(A=Add(A=Add(A=@n3, B=@n4), B=NamedRerouteUsage()), B=NamedRerouteUsage())
  n6 = NamedRerouteUsage()
  n7 = Multiply(A=LinearInterpolate(A=Add(A=@n5, B=@n6), B=NamedRerouteUsage(), Alpha=NamedRerouteUsage()), B=NamedRerouteUsage())
  NamedRerouteDeclaration(None=@n7) {name=EmissiveColorMask, nodeColor=<rgba 1 0.375 0 1>}
  NamedRerouteDeclaration(None=FunctionInput() {bUsePreviewValueAsDefault=true, inputName=MaskMap, sortPriority=1, previewValue=<1 1 1 1>}) {name=InMask, nodeColor=<rgba 1 0.5156 0 1>}
  n8 = TextureSampleParameter2D() {
   parameterName=FX2Mask
   samplerType=SAMPLERTYPE_LinearColor
   texture=/BRCosmetics/Characters/Player/Male/Medium/Bodies/M_MED_KelpLinenCalcium/Textures/T_KelpLinenCalcium_Body_FX2
  }
  NamedRerouteDeclaration(None=LinearInterpolate(A=@n8.RGB, B=NamedRerouteUsage(), Alpha=NamedRerouteUsage())) {name=ReactivityMask, nodeColor=<rgba 0 0.3203 1 1>}
  NamedRerouteDeclaration(None=ComponentMask(None=NamedRerouteUsage()) {g=true}) {name=GMap, nodeColor=<rgba 1 0.1875 0 1>}
  n9 = Multiply(A=NamedRerouteUsage(), B=NamedRerouteUsage())
  NamedRerouteDeclaration(None=LinearInterpolate(A=LinearInterpolate(A=@n7, Alpha=@n9), B=@n7, Alpha=NamedRerouteUsage())) {name=BaseColorMask, nodeColor=<rgba 1 0 0.1641 1>}
  n10 = ChannelMaskParameter(None=NamedRerouteUsage()) {group=KelpLinen, parameterName=SmallDetailsMaskChannel}
  n11 = Add(A=NamedRerouteUsage(), B=LinearInterpolate(Alpha=NamedRerouteUsage()) {constB=0.5})
  n12 = LinearInterpolate(Alpha=ConstantBiasScale(None=Sine(None=@n11))) {constA=8, constB=0.1}
  NamedRerouteDeclaration(None=LinearInterpolate(A=@n10.R, B=Power(Base=@n10.R, Exp=@n12), Alpha=NamedRerouteUsage())) {name=SmallDetailsMask, nodeColor=<rgba 1 0.0234 0 1>}
  n13 = ComponentMask(None=NamedRerouteUsage()) {g=true}
  n14 = OneMinus(None=NamedRerouteUsage())
  NamedRerouteDeclaration(None=LinearInterpolate(A=Multiply(A=@n13, B=@n14), Alpha=NamedRerouteUsage()) {constB=0}) {name=OverallMask, nodeColor=<rgba 0 0.2969 1 1>}
  n15 = NamedRerouteUsage()
  NamedRerouteDeclaration(None=SmoothStep(Value=ChannelMaskParameter(None=@n15) {group=KelpLinen, maskChannel=Blue, parameterName=ShirtMask, defaultValue=<rgba 0 0 1 0>}.R) {constMax=0.25}) {name=ShirtMask, nodeColor=<rgba 1 0.3984 0 1>}
  n16 = ComponentMask(None=NamedRerouteUsage()) {g=true}
  n17 = SphereMask(A=TextureCoordinate(), B=Constant2Vector() {g=0.807, r=0.6525}) {attenuationRadius=0.075, hardnessPercent=25}
  n18 = NamedRerouteDeclaration(None=Ceil(None=NamedRerouteDeclaration(None=@n17) {name=LogoSphereSDF, nodeColor=<rgba 0.1484 1 0 1>})) {name=LogoSphereMask, nodeColor=<rgba 0.1484 1 0 1>}
  NamedRerouteDeclaration(None=SmoothStep(Value=Multiply(A=@n16, B=@n18)) {constMin=0.5}) {name=LogoMask, nodeColor=<rgba 1 0.0234 0 1>}
  NamedRerouteDeclaration(None=ChannelMaskParameter(None=NamedRerouteUsage()) {group=KelpLinen, maskChannel=Blue, parameterName=GradientMaskChannel, defaultValue=<rgba 0 0 1 0>}.R) {name=GradientMask, nodeColor=<rgba 0.9244 0.068 1 1>}
  n19 = Time()
  n20 = Multiply(A=@n19) {constB=2.5}
  n21 = Multiply(A=ConstantBiasScale(None=Sine(None=@n20))) {constB=0.3}
  n22 = Divide(A=@n19, B=ScalarParameter() {defaultValue=15, group=Audio Reactivity, parameterName=FrontendFakeAudio_LoopLength, sortPriority=200})
  n23 = Saturate(None=Multiply(A=Cosine(None=@n22)) {constB=10})
  n24 = Multiply(
   A=Add(A=Multiply(A=@n21, B=@n23), B=Multiply(A=@n23) {constB=0.7})
   B=ScalarParameter() {defaultValue=1, group=Audio Reactivity, parameterName=FrontendFakeAudio_Amplitude, sortPriority=200}
  )
  NamedRerouteDeclaration(None=Multiply(
    A=MaterialFunctionCall!(InAmplitude, InSpectrum, CustomFakeAudioLoop=@n24) {materialFunction=/BRCosmetics/Global/MaterialFunctions/AudioReactivity/MF_AudioReactivity_FrontendLoop}.Amplitude
    B=ScalarParameter() {defaultValue=1, group=KelpLinen, parameterName=EnableReactivity}
   )) {name=Amplitude, nodeColor=<rgba 1 0 0.8203 1>}
  n25 = LinearInterpolate(A=LinearInterpolate(A=NamedRerouteUsage(), Alpha=NamedRerouteUsage()) {constB=0}, Alpha=NamedRerouteUsage()) {constB=0}
  n26 = Add(A=Subtract(A=@n25), B=Power(Base=NamedRerouteUsage()) {constExponent=4})
  NamedRerouteDeclaration(None=SmoothStep(Value=Saturate(None=@n26)) {constMax=0.25}) {name=DotsAmplitudeMask, nodeColor=<rgba 0 0.3203 1 1>}
  n27 = VertexInterpolator(VS=LocalPosition() {localOrigin=InstancePreSkinning}.XYZ)
  n28 = Divide(A=NamedRerouteDeclaration(None=ComponentMask(None=@n27.PS) {b=true, r=true}) {name=PreSkinnedPos, nodeColor=<rgba 1 0.1406 0 1>}) {constB=50}
  n29 = TextureSample(UVs=Panner(Coordinate=@n28) {speedX=-0.005, speedY=-0.005}) {samplerType=SAMPLERTYPE_LinearColor, texture=/Game/Global/Textures/Default/Tileable_Patterns/T_PerlinNoise_Uncompressed}
  n30 = Add(A=Multiply(A=@n28) {constB=0.75}, B=Multiply(A=Subtract(A=@n29.G) {constB=0.5}) {constB=0.1})
  n31 = Subtract(A=TextureSample(UVs=Panner(Coordinate=@n30) {speedX=0.005, speedY=0.005}) {samplerType=SAMPLERTYPE_LinearColor, texture=&T_PerlinNoise_Uncompressed}.G) {constB=0.5}
  n32 = Panner(Coordinate=Add(A=Multiply(A=@n28) {constB=0.3}, B=Multiply(A=@n31) {constB=0.1})) {speedY=-0.02}
  n33 = Subtract(A=NamedRerouteUsage()) {constB=0.2}
  n34 = Add(
   A=Subtract(A=TextureSample(UVs=@n32) {texture=&AtmosphericCloudNoise01}.G, B=Multiply(A=@n33, B=@n33))
   B=Multiply(A=Time()) {constB=0.1}
  )
  NamedRerouteDeclaration(None=ConstantBiasScale(None=Sine(None=@n34))) {name=ShirtNoiseGrayscale, nodeColor=<rgba 1 0.1406 0 1>}
  n35 = NamedRerouteUsage()
  n36 = Divide(A=@n35) {constB=75}
  n37 = ComponentMask(None=@n35) {g=true}
  n38 = Subtract(
   A=TextureSample(UVs=Panner(Coordinate=@n36) {speedX=-0.01, speedY=-0.025}) {samplerType=SAMPLERTYPE_LinearColor, texture=&T_PerlinNoise_Uncompressed}.G
   B=NamedRerouteDeclaration(None=Divide(A=@n37) {constB=250}) {name=TopDownGradient, nodeColor=<rgba 0.4297 1 0 1>}
  )
  n39 = Fresnel() {baseReflectFraction=0, exponent=4}
  n40 = Add(
   A=Subtract(A=Divide(A=@n38) {constB=5}, B=Saturate(None=@n39))
   B=NamedRerouteDeclaration(None=@n2) {name=Time, nodeColor=<rgba 0.875 1 0 1>}
  )
  NamedRerouteDeclaration(None=ConstantBiasScale(None=Sine(None=@n40))) {name=LowFreqNoise, nodeColor=<rgba 0 0.1797 1 1>}
  n41 = NamedRerouteUsage()
  n42 = LinearInterpolate(B=ScalarParameter() {defaultValue=2, parameterName=SmallDetailsSpeed}, Alpha=NamedRerouteUsage()) {constA=5, constB=2}
  n43 = Add(A=LinearInterpolate(A=Divide(A=@n41, B=@n42) {constB=5}, B=@n41, Alpha=NamedRerouteUsage()), B=NamedRerouteUsage())
  NamedRerouteDeclaration(None=Abs(None=Sine(None=@n43))) {name=GradientGrayscaleNoise, nodeColor=<rgba 0.8047 0 1 1>}
  NamedRerouteDeclaration(None=ScalarParameter() {defaultValue=1, group=KelpLinen, parameterName=UseFXMask}) {name=UseFXMask, desc=Used for Eyes and Hair, nodeColor=<rgba 1 0 0.9375 1>}
  n44 = NamedRerouteUsage()
  n45 = LinearInterpolate(A=NamedRerouteUsage(), B=NamedRerouteUsage(), Alpha=NamedRerouteUsage())
  n46 = NamedRerouteUsage()
  n47 = Add(A=NamedRerouteUsage(), B=NamedRerouteUsage())
  n48 = CurveAtlasRowParameter(CurveTime=LinearInterpolate(A=LinearInterpolate(A=@n45, B=@n46, Alpha=@n47), B=NamedRerouteUsage(), Alpha=NamedRerouteUsage())) {
   atlas=&/CA_KelpLinenCalcium
   curve=&/C_KelpLinenCalcium_Shirt
   parameterName=ShirtCurve
  }
  n49 = Desaturation(None=@n48, Fraction=Multiply(A=ScalarParameter() {group=KelpLinen, parameterName=SmallDetailSaturation}, B=NamedRerouteUsage()))
  n50 = NamedRerouteUsage()
  n51 = LinearInterpolate(A=@n49, B=Multiply(A=@n49) {constB=0.4}, Alpha=Saturate(None=Multiply(A=@n50) {constB=1.5})) {constB=0.5}
  n52 = NamedRerouteUsage()
  n53 = ScalarParameter() {group=KelpLinen, parameterName=ShirtSaturation}
  n54 = NamedRerouteUsage()
  n55 = Multiply(
   A=Desaturation(None=LinearInterpolate(A=@n51, B=@n49, Alpha=@n52) {constB=0}, Fraction=Multiply(A=@n53, B=@n54))
   B=LinearInterpolate(B=ScalarParameter() {defaultValue=0.75, group=KelpLinen, parameterName=ShirtBrightness}, Alpha=NamedRerouteUsage()) {constA=1}
  )
  n56 = LinearInterpolate(A=@n55, B=Multiply(A=@n55, B=NamedRerouteUsage()), Alpha=ScalarParameter() {defaultValue=1, group=KelpLinen, parameterName=UseGMap})
  n57 = NamedRerouteUsage()
  n58 = NamedRerouteUsage()
  n59 = Multiply(
   A=Desaturation(None=LinearInterpolate(A=@n44, B=@n56, Alpha=@n57), Fraction=ScalarParameter() {group=KelpLinen, parameterName=Saturation})
   B=Add(A=ScalarParameter() {defaultValue=1, group=KelpLinen, parameterName=EmissiveBoost}, B=Multiply(A=@n58) {constB=5})
  )
  n60 = Multiply(A=@n59) {constB=0.5}
  NamedRerouteDeclaration(None=MaterialFunctionCall(Color_Morning=@n59, Color_Day=@n59, Color_Evening=@n60, Color_Night=@n60) {materialFunction=&Color_Time_of_Day}.Result) {name=OutEmissive, nodeColor=<rgba 0.2422 1 0 1>}
  NamedRerouteDeclaration(None=@n1.EmissiveColor) {name=InEmissive, nodeColor=<rgba 0.0078 1 0 1>}
  NamedRerouteDeclaration(None=@n1.BaseColor) {name=InBaseColor, nodeColor=<rgba 0 0.8828 1 1>}
  NamedRerouteDeclaration(None=LinearInterpolate(A=NamedRerouteUsage(), Alpha=NamedRerouteUsage()) {constB=0}) {name=OutBaseColor, nodeColor=<rgba 0.6017 0.4441 1 1>}
  NamedRerouteDeclaration(None=Ceil(None=@n25)) {name=DotsMask, nodeColor=<rgba 0.6406 0 1 1>}
 ]}
