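# Material function for the KelpLinenCalcium cosmetic backpack that adds animated, audio-reactive glow and spike motion to incoming material attributes using an FX mask.
# An audio reactivity function, fed a synthetic sine-based frontend loop, supplies amplitude.
# Panned Perlin noise, height gradient, Fresnel and time form an oscillating pattern alongside curve-atlas color, shaped by texture and desaturation, then dimmed by time of day.
# The function overrides base color, emissive and pulsing pivot-relative world position offset.
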MaterialFunction /BRCosmetics/Characters/Player/Male/Medium/Bodies/M_MED_KelpLinenCalcium/Materials/MF_KelpLinenCalcium_Backpack
graph {nodes [
  n1 = Time()
  ConstantBiasScale(None=Sine(None=Multiply(A=@n1) {constB=3}))
  n2 = GetMaterialAttributes(None=NamedRerouteUsage()) {attributeGetTypes [#1, #2]}
  FunctionOutput(None=SetMaterialAttributes!(
    MaterialAttributes=@n2.MaterialAttributes
    BaseColor=NamedRerouteUsage()
    EmissiveColor=NamedRerouteUsage()
    WorldPositionOffset=NamedRerouteUsage()
   ) {attributeSetTypes [#1, #2, #3]})
  n3 = Time()
  n4 = Sine(None=Multiply(A=@n3) {constB=2.5})
  n5 = Divide(A=@n3, B=ScalarParameter() {defaultValue=15, group=Audio Reactivity, parameterName=FrontendFakeAudio_LoopLength, sortPriority=200})
  n6 = Saturate(None=Multiply(A=Cosine(None=@n5)) {constB=10})
  n7 = Multiply(A=Multiply(A=ConstantBiasScale(None=@n4)) {constB=0.3}, B=@n6)
  n8 = Multiply(A=@n6) {constB=0.7}
  n9 = MaterialFunctionCall!(
   InAmplitude
   InSpectrum
   CustomFakeAudioLoop=Multiply(
    A=Add(A=@n7, B=@n8)
    B=ScalarParameter() {defaultValue=1, group=Audio Reactivity, parameterName=FrontendFakeAudio_Amplitude, sortPriority=200}
   )
  ) {materialFunction=/BRCosmetics/Global/MaterialFunctions/AudioReactivity/MF_AudioReactivity_FrontendLoop}
  n10 = ScalarParameter() {group=KelpLinen, parameterName=DebugAudio}
  LinearInterpolate(A=@n9.Amplitude, B=@n10, Alpha=@n10)
  NamedRerouteDeclaration(None=FunctionInput() {inputType=FunctionInput_MaterialAttributes}) {name=InMaterial, nodeColor=<rgba 1 0 0.9609 1>}
  NamedRerouteDeclaration(None=FunctionInput() {inputName=FXMask}) {name=InFXMask, nodeColor=<rgba 1 0.5625 0 1>}
  n11 = Subtract(
   A=WorldPosition().XYZ
   B=MaterialFunctionCall() {materialFunction=&ObjectPivotPoint}.Object Pivot Location
  )
  n12 = Multiply(A=Power(Base=NamedRerouteUsage()) {constExponent=4}) {constB=3}
  n13 = ComponentMask(None=NamedRerouteUsage()) {g=true}
  n14 = Multiply(A=Subtract(A=Time(), B=@n13))
  n15 = Multiply(A=ConstantBiasScale(None=Sine(None=@n14)), B=Ceil(None=@n13))
  n16 = Multiply(A=NamedRerouteDeclaration(None=Multiply(A=@n12, B=@n15)) {name=SpikesMask, nodeColor=<rgba 0 1 0.4844 1>}) {constB=0.1}
  NamedRerouteDeclaration(None=Multiply(A=Multiply(A=@n11, B=@n16), B=NamedRerouteUsage())) {name=OutWPO, nodeColor=<rgba 0 1 0.3203 1>}
  NamedRerouteDeclaration(None=Multiply(A=@n9.Amplitude, B=ScalarParameter() {defaultValue=1, group=KelpLinen, parameterName=EnableReactivity})) {name=Amplitude, nodeColor=<rgba 1 0 0.8203 1>}
  n17 = NamedRerouteUsage()
  n18 = Divide(A=@n17) {constB=50}
  n19 = ComponentMask(None=@n17) {g=true}
  n20 = Subtract(
   A=TextureSample(UVs=Panner(Coordinate=@n18) {speedX=-0.01, speedY=-0.025}) {samplerType=SAMPLERTYPE_LinearColor, texture=/Game/Global/Textures/Default/Tileable_Patterns/T_PerlinNoise_Uncompressed}.G
   B=NamedRerouteDeclaration(None=Divide(A=@n19) {constB=250}) {name=TopDownGradient, nodeColor=<rgba 0.4297 1 0 1>}
  )
  n21 = Fresnel() {baseReflectFraction=0, exponent=2}
  n22 = Time()
  n23 = Add(
   A=Subtract(A=Divide(A=@n20) {constB=5}, B=Saturate(None=@n21))
   B=NamedRerouteDeclaration(None=Multiply(A=@n22) {constB=0.05}) {name=Time, nodeColor=<rgba 0.875 1 0 1>}
  )
  NamedRerouteDeclaration(None=ConstantBiasScale(None=Sine(None=@n23))) {name=LowFreqNoise, nodeColor=<rgba 0 0.1797 1 1>}
  n24 = CurveAtlasRowParameter(CurveTime=NamedRerouteUsage()) {
   atlas=&/CA_KelpLinenCalcium
   curve=&/C_KelpLinenCalcium_Shirt
   parameterName=ShirtCurve
  }
  n25 = Saturate(None=LinearInterpolate(Alpha=TextureSample() {texture=/BRCosmetics/Accessories/FORT_Backpacks/Backpack_M_MED_KelpLinenCalcium/Textures/T_KelpLinenCalcium_Backpack_D}.B) {constA=0.2, constB=1.5})
  n26 = Desaturation(
   None=LinearInterpolate(B=Multiply(A=@n24, B=@n25), Alpha=NamedRerouteUsage())
   Fraction=ScalarParameter() {group=KelpLinen, parameterName=Saturation}
  )
  n27 = LinearInterpolate(Alpha=NamedRerouteUsage()) {constB=-10}
  n28 = NamedRerouteUsage()
  n29 = Multiply(
   A=Multiply(A=Desaturation(None=@n26, Fraction=@n27), B=LinearInterpolate(Alpha=@n28) {constA=0.8, constB=1.5})
   B=ScalarParameter() {defaultValue=1, group=KelpLinen, parameterName=EmissiveBoost}
  )
  n30 = Multiply(A=@n29) {constB=0.5}
  NamedRerouteDeclaration(None=MaterialFunctionCall(Color_Morning=@n29, Color_Day=@n29, Color_Evening=@n30, Color_Night=@n30) {materialFunction=&Color_Time_of_Day}.Result) {name=OutEmissive, nodeColor=<rgba 0.2422 1 0 1>}
  NamedRerouteDeclaration(None=@n2.BaseColor) {name=InBaseColor, nodeColor=<rgba 0 0.8828 1 1>}
  n31 = NamedRerouteUsage()
  NamedRerouteDeclaration(None=LinearInterpolate(A=@n31, B=Multiply(A=@n31) {constB=0.2}, Alpha=NamedRerouteUsage()) {constB=0}) {name=OutBaseColor, nodeColor=<rgba 0.6017 0.4441 1 1>}
  n32 = LocalPosition() {localOrigin=InstancePreSkinning}
  NamedRerouteDeclaration(None=ComponentMask(None=VertexInterpolator(VS=@n32.XYZ).PS) {b=true, r=true}) {name=PreSkinnedPos, nodeColor=<rgba 1 0.1406 0 1>}
 ]}
